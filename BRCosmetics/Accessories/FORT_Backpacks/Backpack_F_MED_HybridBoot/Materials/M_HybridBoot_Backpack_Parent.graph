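# Parent material for the HybridBoot backpack cosmetic, adding a localized blue glow to the standard 2023 character shading.
# Glow is added to Emissive between the shared CharacterParent PreFX and PostFX functions.
# PhaseA and PhaseB reroutes build glows from mask-texture channels and pre-skinned-position sphere masks, PhaseB adding a normal-based directional highlight.
# A GlowAnimation scalar and Tech Art hilt mask are declared; the Time-based AnimationSpeed stepper is unconnected.
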Material /BRCosmetics/Accessories/FORT_Backpacks/Backpack_F_MED_HybridBoot/Materials/M_HybridBoot_Backpack_Parent
graph {
 properties {
  bAutomaticallySetUsageInEditor=false
  bUseMaterialAttributes=true
  bUsedWithMeshDeformer=true
  bUsedWithSkeletalMesh=true
  pixelDepthOffsetMode=PDOM_Legacy
  shadingModel=MSM_FromMaterialExpression
  usedShadingModels=Default Lit | Subsurface | Clear Coat | Subsurface Profile | Cloth
 }
 nodes [
  n1 = Time()
  n2 = ScalarParameter() {defaultValue=1, group=HybridBoot, parameterName=AnimationSpeed}
  Fmod(A=Floor(None=Multiply(A=@n1, B=@n2)), B=Constant() {r=3})
  n3 = MaterialFunctionCall() {materialFunction=&MF_CharacterParent_2023_PreFX}
  n4 = GetMaterialAttributes(None=@n3.Result) {attributeGetTypes [#1, #2]}
  n5 = TextureSample() {samplerType=SAMPLERTYPE_Masks, texture=/BRCosmetics/Accessories/FORT_Backpacks/Backpack_F_MED_HybridBoot/Textures/T_HybridBoot_Backpack_TA}
  n6 = Add(A=Multiply(A=@n5.B) {constB=50}, B=NamedRerouteUsage())
  n7 = Saturate(None=Subtract(A=NamedRerouteUsage()))
  n8 = NamedRerouteUsage()
  n9 = Saturate(None=NamedRerouteUsage())
  n10 = Multiply(A=Add(A=Multiply(A=@n6, B=@n7), B=Multiply(A=@n8, B=@n9)), B=NamedRerouteUsage())
  MaterialFunctionCall(
   Material=SetMaterialAttributes!(MaterialAttributes=@n4.MaterialAttributes, Emissive Color=Add(A=@n4.Emissive Color, B=@n10), Base Color) {attributeSetTypes [#1, #2]}
   M=@n3.M
   Pre Skinned Local Position=@n3.M
   Distance Blend=@n3.M
   Fresnel=@n3.M
  ) {materialFunction=&MF_CharacterParent_2023_PostFX}
  Max(A=@n5.G, B=@n5.G)
  NamedRerouteDeclaration(None=@n3.Pre skinned Local Position) {name=PreSkinnedLocalPosition, nodeColor=<rgba 0 1 0.1797 1>}
  n11 = SphereMask(
   A=NamedRerouteUsage()
   B=Constant4Vector() {constant=<rgba 5.4773 -3.5749 6.0093 1>}.RGB
   Radius=Constant() {r=4}
   Hardness=Constant() {r=0.0357}
  )
  n12 = Constant4Vector() {constant=<rgba 0 0.3911 1 1>}
  n13 = TextureSample() {samplerType=SAMPLERTYPE_Masks, texture=&T_HybridBoot_Backpack_TA}
  n14 = Multiply(
   A=SphereMask(
    A=NamedRerouteUsage()
    B=Constant4Vector() {constant=<rgba 5.4773 -3.5749 6.0093 1>}.RGB
    Radius=Constant() {r=4.3321}
    Hardness=Constant() {r=1}
   )
   B=@n13.G
  )
  n15 = Multiply(
   A=Constant4Vector() {constant=<rgba 0 0.3876 0.9911 1>}.RGB
   B=ScalarParameter() {defaultValue=100, group=HybridBoot, parameterName=H brightness}
  ) {constB=20}
  NamedRerouteDeclaration(None=Add(A=Multiply(A=@n11, B=@n12.RGB), B=Multiply(A=@n14, B=@n15))) {name=PhaseA, nodeColor=<rgba 0 1 0.5781 1>}
  n16 = TextureSample() {samplerType=SAMPLERTYPE_Masks, texture=&T_HybridBoot_Backpack_TA}
  n17 = Power(Base=@n16.R) {constExponent=2.2}
  n18 = Constant() {r=2.2}
  n19 = VertexNormalWS()
  n20 = Saturate(None=DotProduct(
    A=Transform(None=@n19) {transformSourceType=TRANSFORMSOURCE_World, transformType=TRANSFORM_Local}
    B=Constant4Vector() {constant=<rgba -0.504 0 -1 -1>}.RGB
   ))
  n21 = Constant() {r=4}
  n22 = Add(
   A=Power(Base=Multiply(A=@n17, B=@n18) {constB=3}, Exp=Constant() {r=10}) {constExponent=3}
   B=Multiply(A=Power(Base=@n20, Exp=@n21), B=Constant() {r=15})
  )
  n23 = Constant4Vector() {constant=<rgba 0 0.3911 1 1>}
  n24 = NamedRerouteUsage()
  n25 = Constant4Vector() {constant=<rgba 5.4773 -3.5749 6.0093 1>}
  n26 = Constant() {r=7.75}
  n27 = Constant() {r=0.5}
  NamedRerouteDeclaration(None=Multiply(A=Multiply(A=@n22, B=@n23.RGB), B=SphereMask(A=@n24, B=@n25.RGB, Radius=@n26, Hardness=@n27))) {name=PhaseB, nodeColor=<rgba 1 0.2578 0 1>}
  NamedRerouteDeclaration(None=ScalarParameter() {parameterName=GlowAnimation}) {name=Animation loop, nodeColor=<rgba 0 0.0625 1 1>}
  NamedRerouteDeclaration(None=ComponentMask(None=@n3.Tech Art Mask) {b=true}) {name=HiltMask, nodeColor=<rgba 1 0.4219 0 1>}
  OneMinus(None=NamedRerouteUsage())
 ]
}
