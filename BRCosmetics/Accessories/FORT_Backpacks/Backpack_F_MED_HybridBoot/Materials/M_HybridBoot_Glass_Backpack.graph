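# Translucent, per-pixel lit glass material for the HybridBoot backpack cosmetic, adding glowing blue edges to a configurable glass base.
# The base drives emissive and opacity from Fresnel over a noise-perturbed normal, with optional masked diffuse.
# Edge glow combines texture masks, pre-skinned local-space sphere masks, a directional normal term and a GlowAnimation parameter through named reroutes.
# Switches add thin film, vertex hiding, camera dither and a shadow-offset fix, alongside LOD and black-hole vertex offsets.
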Material /BRCosmetics/Accessories/FORT_Backpacks/Backpack_F_MED_HybridBoot/Materials/M_HybridBoot_Glass_Backpack
graph {
 properties {
  bAutomaticallySetUsageInEditor=false
  bEnableResponsiveAA=true
  bScreenSpaceReflections=true
  bUseMaterialAttributes=true
  bUsedWithMeshDeformer=true
  bUsedWithSkeletalMesh=true
  blendMode=BLEND_Translucent
  ditherOpacityMask=true
  pixelDepthOffsetMode=PDOM_Legacy
  translucencyDirectionalLightingIntensity=5
  translucencyLightingMode=TLM_SurfacePerPixelLighting
  usedShadingModels=Default Lit
 }
 nodes [
  Constant()
  n1 = VectorParameter() {group=HybridBoot, parameterName=GlassEdgeEmissiveColor, defaultValue=<rgba 0.005 0.1114 0.5 1>}
  n2 = Constant() {r=6}
  n3 = Multiply(A=NamedRerouteUsage(), B=ScalarParameter() {defaultValue=0.6, group=HybridBoot, parameterName="Glass edge  Emissive"})
  n4 = NamedRerouteUsage()
  n5 = Constant() {r=20}
  n6 = VertexInterpolator(VS=LocalPosition() {localOrigin=InstancePreSkinning}.XYZ)
  n7 = Constant4Vector() {constant=<rgba 1.2994 -3.1998 -0.9667 1>}
  n8 = Constant() {r=6}
  n9 = Constant() {r=-4.5354}
  n10 = Multiply(A=Power(Base=SphereMask(A=@n6.PS, B=@n7.RGB, Radius=@n8, Hardness=@n9)), B=Constant() {r=120})
  n11 = LocalPosition() {localOrigin=InstancePreSkinning}
  n12 = Multiply(
   A=SphereMask(
    A=VertexInterpolator(VS=@n11.XYZ).PS
    B=Constant4Vector() {constant=<rgba -0.3929 -3.1998 -3.7531 1>}.RGB
    Radius=Constant() {r=14.6578}
    Hardness=Constant() {r=-0.5813}
   )
   B=Constant() {r=0.5}
  )
  n13 = Transform(None=VertexNormalWS()) {transformSourceType=TRANSFORMSOURCE_World, transformType=TRANSFORM_Local}
  n14 = Constant4Vector() {constant=<rgba 0.6684 0 1 1>}
  n15 = Multiply(A=Power(Base=DotProduct(A=@n13, B=@n14.RGB), Exp=Constant() {r=4}), B=Constant() {r=15})
  n16 = Fresnel(ExponentIn=Constant() {r=5}, BaseReflectFractionIn=Constant() {r=0.03})
  n17 = Add(A=Add(A=Multiply(A=@n4, B=@n5), B=Add(A=@n10, B=@n12)), B=Add(A=Add(A=@n15, B=@n16), B=Constant() {r=0.05}))
  NamedRerouteDeclaration(None=Multiply(A=Max(A=@n3, B=@n17), B=Multiply(A=@n1.RGB, B=@n2))) {name=EmissiveColor, nodeColor=<rgba 1 0.5625 0 1>}
  n18 = SphereMask(
   A=VertexInterpolator(VS=LocalPosition() {localOrigin=InstancePreSkinning}.XYZ).PS
   B=Constant4Vector() {constant=<rgba 13.433 -3.1998 7.5325 1>}.RGB
   Radius=Constant() {r=4.4863}
   Hardness=Constant() {r=-3}
  )
  n19 = SphereMask(
   A=VertexInterpolator(VS=LocalPosition() {localOrigin=InstancePreSkinning}.XYZ).PS
   B=Constant4Vector() {constant=<rgba 4.3458 -3.1998 13.9234 1>}.RGB
   Radius=Constant() {r=4.4863}
   Hardness=Constant() {r=-3}
  )
  NamedRerouteDeclaration(None=Max(A=Power(Base=@n18), B=Power(Base=@n19))) {name=glow Point, nodeColor=<rgba 1 0 0.3047 1>}
  n20 = ScalarParameter() {group=HybridBoot, parameterName=GlowAnimation}
  NamedRerouteDeclaration(None=Saturate(None=Subtract(A=@n20))) {name=AnimationLoopB, nodeColor=<rgba 0.7344 0 1 1>}
  n21 = TextureSample() {
   samplerSource=SSM_Wrap_WorldGroupSettings
   samplerType=SAMPLERTYPE_Masks
   texture=/BRCosmetics/Accessories/FORT_Backpacks/Backpack_F_MED_HybridBoot/Textures/T_HybridBoot_Backpack_TA
  }
  NamedRerouteDeclaration(None=@n21.B) {name=HText, nodeColor=<rgba 0 1 0.1328 1>}
  NamedRerouteDeclaration(None=Saturate(None=@n20)) {name=AnimationLoopA, nodeColor=<rgba 0.7344 0 1 1>}
  n22 = TextureSample() {
   samplerSource=SSM_Wrap_WorldGroupSettings
   samplerType=SAMPLERTYPE_Masks
   texture=&T_HybridBoot_Backpack_TA
  }
  Power(Base=@n22.R)
  n23 = Constant3Vector()
  n24 = Multiply(A=SkyAtmosphereLightDirection(), B=ScalarParameter() {defaultValue=-1, group=Glass, parameterName=ShadowBufferOffset}) {constB=-1}
  n25 = ScalarParameter() {parameterName=ThinFilmOpacity}
  n26 = TextureSampleParameter2D(UVs=TextureCoordinate()) {group=Mask Texture, parameterName=Diffuse Texture with Alpha Mask, sortPriority=10, texture=&T_BlacksRGB}
  n27 = MaterialFunctionCall(Float3=TextureSampleParameter2D() {parameterName=Input S, texture=/Game/Global/Textures/Default/Blanks/T_BlackUserInterface}.RGB) {materialFunction=&BreakOutFloat3Components}
  n28 = StaticBoolParameter() {group=Specular, parameterName=SwizzleRoughnessToGreen}
  n29 = MaterialFunctionCall(Float3=MaterialFunctionCall(X=@n27.R, Y=StaticSwitch(True=@n27.B, False=@n27.B, Value=@n28), Z=StaticSwitch(True=@n27.G, False=@n27.G, Value=@n28)) {materialFunction=&MakeFloat3}.Result) {materialFunction=&BreakOutFloat3Components}
  n30 = ScalarParameter() {defaultValue=1, parameterName=Fresnel Exponent}
  n31 = ObjectBounds()
  n32 = PreSkinnedLocalBounds()
  n33 = Saturate(None=Floor(None=ScalarParameter() {parameterName=Use Pre-Skinned Bounds, sliderMax=1}))
  n34 = ComponentMask(None=Divide(A=PreSkinnedPosition(), B=LinearInterpolate(A=@n31, B=@n32.Half Extents, Alpha=@n33))) {b=true, g=true}
  n35 = ScalarParameter() {defaultValue=0.2, parameterName=WaterTiling}
  n36 = TextureSampleParameter2D(UVs=VertexInterpolator(VS=Multiply(A=@n34, B=@n35)).PS) {
   parameterName=Noise Normal
   samplerSource=SSM_Wrap_WorldGroupSettings
   samplerType=SAMPLERTYPE_Normal
   texture=/Game/Global/Textures/Default/Noise/T_Noise_Normal_01
  }
  n37 = VectorParameter() {parameterName=Normal Intensity}
  n38 = TextureSampleParameter2D() {
   group=Mask Texture
   parameterName=Normals
   samplerSource=SSM_Wrap_WorldGroupSettings
   samplerType=SAMPLERTYPE_Normal
   texture=/Engine/EngineMaterials/FlatNormal
  }
  n39 = Normalize(VectorInput=Add(A=Multiply(A=@n36.RGB, B=@n37.RGB), B=@n38.RGB))
  n40 = Transform(None=@n39)
  n41 = Clamp(None=Multiply(A=Fresnel(ExponentIn=@n30, Normal=@n40), B=ScalarParameter() {defaultValue=1, parameterName=Glass thickness}))
  n42 = LinearInterpolate(
   A=ScalarParameter() {parameterName=Fresnel Inner Emissive}
   B=ScalarParameter() {defaultValue=0.85, parameterName=Fresnel Outer Emissive}
   Alpha=@n41
  )
  n43 = VectorParameter() {parameterName=EmissiveColor}
  n44 = TextureSampleParameter2D() {parameterName=Input FX, texture=&T_BlackUserInterface}
  n45 = ScalarParameter() {parameterName=Add Custom Opacity, sortPriority=10}
  n46 = ScalarParameter() {parameterName=Fresnel Inner Transparency}
  n47 = ScalarParameter() {defaultValue=0.85, parameterName=Fresnel Outer Transparency}
  n48 = LinearInterpolate(
   A=Add(A=Multiply(A=@n44.B, B=@n45), B=LinearInterpolate(A=@n46, B=@n47, Alpha=@n41))
   B=@n44.B
   Alpha=ScalarParameter() {parameterName=Use Custom Opacity}
  )
  n49 = MakeMaterialAttributes!(
   BaseColor=StaticSwitchParameter(True=@n26.RGB, False=VectorParameter() {parameterName=Base Color}.RGB) {parameterName="Use Diffuse Texture for Color [ignores alpha channel]"}
   Metallic=LinearInterpolate(A=ScalarParameter() {defaultValue=0.1, parameterName=Metallic}, B=@n29.G, Alpha=ScalarParameter() {parameterName=Use Custom Metallic})
   Specular=LinearInterpolate(A=ScalarParameter() {defaultValue=1, parameterName=Specular}, B=@n29.R, Alpha=ScalarParameter() {parameterName=Use Custom Specular})
   Roughness=LinearInterpolate(A=ScalarParameter() {parameterName=Roughness}, B=@n29.B, Alpha=ScalarParameter() {parameterName=Use Custom Roughness})
   Anisotropy
   EmissiveColor=Multiply(
    A=Multiply(A=@n42, B=@n43.RGB)
    B=TextureSampleParameter2D() {parameterName=EmissiveMask, samplerType=SAMPLERTYPE_Masks, texture=&T_BlackMask}.RGB
   )
   Opacity=@n48
   OpacityMask=@n48
   Normal=@n39
   Tangent
   WorldPositionOffset
   SubsurfaceColor
   ClearCoat
   ClearCoatRoughness
   AmbientOcclusion
   Refraction
   CustomizedUVs_0
   CustomizedUVs_1
   CustomizedUVs_2
   CustomizedUVs_3
   CustomizedUVs_4
   CustomizedUVs_5
   CustomizedUVs_6
   CustomizedUVs_7
   PixelDepthOffset
   ShadingModel
   Displacement
  )
  n50 = GetMaterialAttributes(None=@n49) {attributeGetTypes [
    #1
    #2
    #3
    #4
    #5
   ]}
  n51 = Multiply(A=@n26.A, B=ScalarParameter() {defaultValue=1, group=Mask Texture, parameterName=Alpha Channel Mask Opacity})
  n52 = ScalarParameter() {defaultValue=0.4, group=Mask Texture, parameterName=Masked Area Roughness, sortPriority=4}
  n53 = Multiply(A=LinearInterpolate(A=@n50.Opacity, Alpha=@n51), B=Constant() {r=0.05})
  n54 = Power(Base=@n21.R)
  n55 = Power(Base=Multiply(A=@n54, B=Constant() {r=1.8}) {constB=2}, Exp=Constant() {r=2.5})
  n56 = Add(A=@n54, B=NamedRerouteUsage())
  n57 = Constant() {r=1.8}
  n58 = NamedRerouteDeclaration(None=Power(Base=Multiply(A=@n56, B=@n57) {constB=2}, Exp=Constant() {r=2.5})) {name=Glass edge, nodeColor=<rgba 1 0 0.1875 1>}
  n59 = NamedRerouteUsage()
  n60 = Saturate(None=Add(A=Add(A=@n53, B=@n55), B=Multiply(A=@n58, B=@n59)))
  n61 = Multiply(A=Multiply(A=NamedRerouteUsage(), B=Constant4Vector() {constant=<rgba 0.02 0.7118 2 1>}.RGB), B=NamedRerouteUsage())
  n62 = NamedRerouteUsage()
  n63 = SetMaterialAttributes!(
   MaterialAttributes=@n50.MaterialAttributes
   Base Color=Saturate(None=LinearInterpolate(A=@n50.Base Color, B=@n26.RGB, Alpha=@n51))
   Metallic=LinearInterpolate(A=@n50.Metallic, B=ScalarParameter() {group=Mask Texture, parameterName=Masked Area Metallic, sortPriority=2}, Alpha=@n51)
   Specular=LinearInterpolate(A=@n50.Specular, B=ScalarParameter() {defaultValue=0.5, group=Mask Texture, parameterName=Masked Area Specular, sortPriority=3}, Alpha=@n51)
   Roughness=Saturate(None=LinearInterpolate(A=@n50.Roughness, B=@n52, Alpha=@n51))
   Opacity=@n60
   Opacity Mask=@n60
   Emissive Color=Multiply(A=Add(A=@n61, B=@n62), B=NamedRerouteUsage())
  ) {attributeSetTypes [
    #1
    #2
    #3
    #4
    #5
    #6
    #7
   ]}
  n64 = StaticSwitchParameter(True=@n63, False=@n49) {group=Mask Texture, parameterName=Use Diffuse Texture with Alpha Mask, sortPriority=-50}
  n65 = Multiply(A=Subtract(A=ObjectPositionWS(), B=WorldPosition().XYZ), B=ScalarParameter() {parameterName=HT_CrunchVerts})
  n66 = VectorParameter() {parameterName=WorldPostionOfBlackHole, defaultValue=<rgba -30000 25536 1952 0>}
  n67 = MaterialFunctionCall!(
   World Position=WorldPosition().XYZ
   DistanceCullFadeAnimationOffset=MaterialFunctionCall() {materialFunction=/Game/Packages/Fortress_SharedMaterials/Base_Material_Functions/PickupEffectDistanceCullOverride}.Result
   Use actor center instead of pivot point
  ) {materialFunction=/Game/Packages/Fortress_SharedMaterials/Base_Material_Functions/LODEffect}
  n68 = WorldPosition()
  n69 = Add(
   A=MaterialFunctionCall!(WorldPostionOfBlackHole=AppendVector(A=@n66.RGB, B=@n66.RGB), World Position=Add(A=@n67.Result, B=@n68.XYZ)) {materialFunction=/Game/Packages/Fortnite_UniqueMaterials_1/BlackHoleFunction}.Result
   B=@n67.Result
  )
  n70 = SetMaterialAttributes!(
   MaterialAttributes=StaticSwitchParameter(
    True=MaterialFunctionCall!(Alpha=@n25, Material Attributes=@n64) {materialFunction=/Game/Weapons/WeaponSkins/MaterialFunction/ThinFilmFunction}.Result
    False=@n64
   ) {parameterName=useThinFilm}
   World Position Offset=StaticSwitchParameter(True=Add(A=@n65, B=@n69), False=@n69) {parameterName=bUseVertexHide}
  ) {attributeSetTypes [#8]}
  n71 = MaterialFunctionCall(In=StaticSwitchParameter(
    True=MaterialFunctionCall!(In=@n70) {materialFunction=/Game/Packages/Fortress_SharedMaterials/Base_Material_Functions/BaseCustomizationFunctions/MF_FN_Vehicle_CameraDither}.Result
    False=@n70
   ) {parameterName=UseCameraDither}) {materialFunction=&MF_TechArtOverrides}
  StaticSwitchParameter(
   True=SetMaterialAttributes(MaterialAttributes=@n71.Result, World Position Offset=ShadowReplace(Default=@n23, Shadow=@n24)) {attributeSetTypes [#8]}
   False=@n71.Result
  ) {desc=Fix Glass Artifacts When Cast Dynamic Shadow as Mask is ON, group=Glass, parameterName=Use_GlassCastShadowOffsetFix, sortPriority=0}
 ]
}
